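# Control Rig for the Space Wanderer medium male body that corrects the front costume straps when crouching, judging by its name.
# A sequence runs spherical pose readers on each thigh, whose outputs weight offsets on the matching dyn_strapFront_01 bone.
# The shared "Set Transforms for Bone" function adds translate, Euler rotate and scale offsets to the current local transform.
# Front-flap remaps from thigh angles, an upper-arm corrective and spare strap variants remain unconnected to execution.
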ControlRigBlueprint /BRCosmetics/Characters/Player/Male/Medium/Bodies/M_MED_Space_Wanderer/Meshes/M_MED_Space_Wanderer_CtrlRig/M_MED_Space_Wander_Crouch_CtrlRig
graphs [
 ContainedGraph {nodes [
   n1 = RigVMFunctionEntryNode()
   RigUnit_GetTransform = &RigUnit_GetTransform(Space=LocalSpace, bInitial=false, Item=@n1.Joint)
   n2 = &RigVMFunction_MathQuaternionToEuler(RotationOrder=ZYX, Value=@RigUnit_GetTransform.Transform.Rotation)
   > RigUnit_SetTransform = &RigUnit_SetTransform(
    Space=LocalSpace
    Value=(Rotation=(X=0.0132,Y=0.4099,Z=-0.0222,W=0.9118),Translation=(X=23.6045,Y=-6.37,Z=140.1115),Scale3D=(X=1,Y=1,Z=1))
    bInitial=false
    Item=@n1.Joint
    Weight=@n1.Weight
    Value.Translation.Z=&RigVMFunction_MathFloatAdd(B=@n1.Translate Z, A=@RigUnit_GetTransform.Transform.Translation.Z) {template="Add::Execute(in A,in B,out Result)"}.Result
    Value.Scale3D.Z=&RigVMFunction_MathFloatAdd(B=@n1.Scale Z, A=@RigUnit_GetTransform.Transform.Scale3D.Z) {template="Add::Execute(in A,in B,out Result)"}.Result
    Value.Translation.Y=&RigVMFunction_MathFloatAdd(B=@n1.Translate Y, A=@RigUnit_GetTransform.Transform.Translation.Y) {template="Add::Execute(in A,in B,out Result)"}.Result
    Value.Scale3D.Y=&RigVMFunction_MathFloatAdd(B=@n1.Scale Y, A=@RigUnit_GetTransform.Transform.Scale3D.Y) {template="Add::Execute(in A,in B,out Result)"}.Result
    Value.Translation.X=&RigVMFunction_MathFloatAdd(B=@n1.Translate X, A=@RigUnit_GetTransform.Transform.Translation.X) {template="Add::Execute(in A,in B,out Result)"}.Result
    Value.Scale3D.X=&RigVMFunction_MathFloatAdd(B=@n1.Scale X, A=@RigUnit_GetTransform.Transform.Scale3D.X) {template="Add::Execute(in A,in B,out Result)"}.Result
    Value.Rotation=&RigVMFunction_MathQuaternionFromEuler(
     Euler=(X=0,Y=0,Z=0)
     Euler.X=&RigVMFunction_MathFloatAdd(B=@n1.Rotate X, A=@n2.Result.X) {template="Add::Execute(in A,in B,out Result)"}.Result
     Euler.Y=&RigVMFunction_MathFloatAdd(B=@n1.Rotate Y, A=@n2.Result.Y) {template="Add::Execute(in A,in B,out Result)"}.Result
     Euler.Z=&RigVMFunction_MathFloatAdd(B=@n1.Rotate Z, A=@n2.Result.Z) {template="Add::Execute(in A,in B,out Result)"}.Result
    ).Result
   )
   > RigVMFunctionReturnNode()
  ]}
 RigVMFunctionLibrary {nodes [Set Transforms for Bone = RigVMCollapseNode(
    ExecuteContext=()
    Joint=()
    Rotate X=0
    Rotate Y=0
    Rotate Z=0
    Scale X=0
    Scale Y=0
    Scale Z=0
    Translate X=0
    Translate Y=0
    Translate Z=0
    Weight=1
   )]}
 RigVMModel {nodes [
   n1 = &RigVMFunction_MathQuaternionToEuler(RotationOrder=ZYX, Value=&RigUnit_GetRelativeBoneTransform(Bone=thigh_l, Space=pelvis).Transform.Rotation)
   n2 = &RigVMFunction_MathQuaternionToEuler(RotationOrder=ZYX, Value=&RigUnit_GetRelativeBoneTransform(Bone=thigh_l, Space=pelvis).Transform.Rotation)
   /Script/ControlRig.RigUnit_ModifyBoneTransforms(
    BoneToModify=((Bone="dyn_frontFlapMain_Base",Transform=(Rotation=(X=0,Y=0,Z=0,W=1),Translation=(X=0,Y=0,Z=0),Scale3D=(X=1,Y=1,Z=1))))
    ExecutePin=()
    Mode=AdditiveLocal
    Weight=0.2
    WeightMaximum=1
    WeightMinimum=0
    BoneToModify.0.Transform.Translation.Y=&RigVMFunction_MathFloatRemap(SourceMaximum=-117, SourceMinimum=-36, TargetMaximum=3, TargetMinimum=0, bClamp=true, Value=@n1.Result.Z).Result
    BoneToModify.0.Transform.Translation.X=&RigVMFunction_MathFloatRemap(SourceMaximum=-117, SourceMinimum=-36, TargetMaximum=-4, TargetMinimum=0, bClamp=true, Value=@n1.Result.Z).Result
    BoneToModify.0.Transform.Translation.Z=&RigVMFunction_MathFloatRemap(SourceMaximum=-117, SourceMinimum=-36, TargetMaximum=0, TargetMinimum=0, bClamp=true, Value=@n1.Result.Z).Result
    BoneToModify.0.Transform.Rotation=&RigVMFunction_MathQuaternionFromEuler(
     Euler=(X=0,Y=0,Z=0)
     Euler.X=&RigVMFunction_MathFloatRemap(SourceMaximum=-117, SourceMinimum=-36, TargetMaximum=-10, TargetMinimum=0, bClamp=true, Value=@n2.Result.Z).Result
     Euler.Z=&RigVMFunction_MathFloatRemap(SourceMaximum=-117, SourceMinimum=-36, TargetMaximum=-70, TargetMinimum=0, bClamp=true, Value=@n2.Result.Z).Result
     Euler.Y=&RigVMFunction_MathFloatRemap(SourceMaximum=-117, SourceMinimum=-36, TargetMaximum=50, TargetMinimum=0, bClamp=true, Value=@n2.Result.Z).Result
    ).Result
   ) {is=struct RigVMUnitNode}
   n3 = /Script/RigVM.RigVMFunction_MathFloatAbs(Value=&RigVMFunction_MathQuaternionToEuler(RotationOrder=ZYX, Value=&RigUnit_GetRelativeBoneTransform(Bone=thigh_r, Space=pelvis).Transform.Rotation).Result.Z) {is=struct RigVMUnitNode, template="Absolute::Execute(in Value,out Result)"}
   > &RigUnit_ModifyBoneTransforms(
    BoneToModify=((Bone="dyn_frontFlapMain_Base",Transform=(Rotation=(X=0,Y=0,Z=0,W=1),Translation=(X=0,Y=0,Z=0),Scale3D=(X=1,Y=1,Z=1))))
    Mode=AdditiveLocal
    Weight=1
    WeightMaximum=1
    WeightMinimum=0
    BoneToModify.0.Transform.Translation.X=&RigVMFunction_MathFloatRemap(SourceMaximum=78, SourceMinimum=140, TargetMaximum=0, TargetMinimum=0, bClamp=true, Value=@n3.Result).Result
    BoneToModify.0.Transform.Translation.Y=&RigVMFunction_MathFloatRemap(SourceMaximum=78, SourceMinimum=140, TargetMaximum=4, TargetMinimum=0, bClamp=true, Value=@n3.Result).Result
    BoneToModify.0.Transform.Translation.Z=&RigVMFunction_MathFloatRemap(SourceMaximum=78, SourceMinimum=140, TargetMaximum=0, TargetMinimum=0, bClamp=true, Value=@n3.Result).Result
    BoneToModify.0.Transform.Rotation=&RigVMFunction_MathQuaternionFromEuler(
     Euler=(X=0,Y=0,Z=0)
     Euler.Y=&RigVMFunction_MathFloatRemap(SourceMaximum=78, SourceMinimum=140, TargetMaximum=-50, TargetMinimum=0, bClamp=true, Value=@n3.Result).Result
     Euler.Z=&RigVMFunction_MathFloatRemap(SourceMaximum=78, SourceMinimum=140, TargetMaximum=-90, TargetMinimum=0, bClamp=true, Value=@n3.Result).Result
     Euler.X=&RigVMFunction_MathFloatRemap(SourceMaximum=78, SourceMinimum=140, TargetMaximum=0, TargetMinimum=0, bClamp=true, Value=@n3.Result).Result
    ).Result
   )
   RigUnit_BeginExecution = &RigUnit_BeginExecution()
   n4 = /Script/ControlRig.RigUnit_SequenceExecution(ExecuteContext=@RigUnit_BeginExecution.ExecutePin) {is=struct RigVMUnitNode}
   n5 = &RigUnit_SphericalPoseReader(
    ActiveRegionScaleFactors=(PositiveWidth=3.5,NegativeWidth=1,PositiveHeight=3.5,NegativeHeight=4)
    ActiveRegionSize=0.1
    Debug=(bDrawDebug=True,DebugScale=25,DebugSegments=20,DebugThickness=0.25)
    DriverAxis=(X=-1,Y=0,Z=0)
    DriverItem=(Type=Bone,Name="thigh_l")
    FalloffRegionScaleFactors=(PositiveWidth=1,NegativeWidth=0.9,PositiveHeight=1,NegativeHeight=1)
    FalloffSize=0.45
    FlipHeightScaling=true
    FlipWidthScaling=true
    OptionalParentItem=(Type=Bone,Name="None")
    RotationOffset=(X=90,Y=180,Z=0)
    ExecutePin=@n4.A
   )
   > Set Transforms for Bone_1_1_1_4_3 = RigVMFunctionReferenceNode(
    Joint=(Type=Bone,Name="dyn_strapFront_01_l")
    Rotate X=0
    Rotate Y=30
    Rotate Z=-90
    Scale X=0
    Scale Y=0
    Scale Z=0
    Translate X=-12
    Translate Y=0
    Translate Z=0
    Weight=@n5.OutputParam
   )
   n6 = &RigUnit_SphericalPoseReader(
    ActiveRegionScaleFactors=(PositiveWidth=3.5,NegativeWidth=0.5,PositiveHeight=3.5,NegativeHeight=4)
    ActiveRegionSize=0.1
    Debug=(bDrawDebug=True,DebugScale=25,DebugSegments=20,DebugThickness=0.25)
    DriverAxis=(X=1,Y=0,Z=0)
    DriverItem=(Type=Bone,Name="thigh_r")
    FalloffRegionScaleFactors=(PositiveWidth=1,NegativeWidth=1.1,PositiveHeight=1,NegativeHeight=1)
    FalloffSize=0.45
    FlipHeightScaling=true
    FlipWidthScaling=true
    OptionalParentItem=(Type=Bone,Name="None")
    RotationOffset=(X=-90,Y=0,Z=0)
    ExecutePin=@n4.B
   )
   > Set Transforms for Bone_1_1_1_4_1_1 = RigVMFunctionReferenceNode(
    Joint=(Type=Bone,Name="dyn_strapFront_01_r")
    Rotate X=20
    Rotate Y=30
    Rotate Z=-90
    Scale X=0
    Scale Y=0
    Scale Z=0
    Translate X=-14
    Translate Y=0
    Translate Z=0
    Weight=@n6.OutputParam
   )
   n7 = &RigUnit_SphericalPoseReader(
    ActiveRegionScaleFactors=(PositiveWidth=1.25,NegativeWidth=1,PositiveHeight=1,NegativeHeight=1)
    ActiveRegionSize=0.3
    Debug=(bDrawDebug=True,DebugScale=25,DebugSegments=20,DebugThickness=0.25)
    DriverAxis=(X=1,Y=0,Z=0)
    DriverItem=(Type=Bone,Name="upperarm_l")
    ExecutePin=()
    FalloffRegionScaleFactors=(PositiveWidth=0.6,NegativeWidth=1.3,PositiveHeight=1,NegativeHeight=1)
    FalloffSize=0.4
    FlipHeightScaling=true
    FlipWidthScaling=true
    OptionalParentItem=(Type=Bone,Name="spine_04")
    RotationOffset=(X=30,Y=15,Z=50)
   )
   > Set Transforms for Bone_1_1_1_1 = RigVMFunctionReferenceNode(
    Joint=(Type=Bone,Name="dyn_arm_l_upperarm_l")
    Rotate X=0
    Rotate Y=0
    Rotate Z=0
    Scale X=0.15
    Scale Y=0
    Scale Z=0
    Translate X=0
    Translate Y=0
    Translate Z=0
    Weight=@n7.OutputParam
   )
   Set Transforms for Bone_1_1_1 = RigVMFunctionReferenceNode(
    ExecuteContext=()
    Joint=(Type=Bone,Name="dyn_strapFront_01_l")
    Rotate X=-90
    Rotate Y=0
    Rotate Z=0
    Scale X=0
    Scale Y=0
    Scale Z=0
    Translate X=0
    Translate Y=0
    Translate Z=0
    Weight=1
   )
   Set Transforms for Bone_1_1_1_2 = RigVMFunctionReferenceNode(
    ExecuteContext=()
    Joint=(Type=Bone,Name="dyn_strapFront_01_l")
    Rotate X=-40
    Rotate Y=25
    Rotate Z=0
    Scale X=0
    Scale Y=0
    Scale Z=0
    Translate X=0
    Translate Y=-7
    Translate Z=-5
    Weight=1
   )
   Set Transforms for Bone_1_1_1_3 = RigVMFunctionReferenceNode(
    ExecuteContext=()
    Joint=(Type=Bone,Name="dyn_strapFront_01_l")
    Rotate X=-40
    Rotate Y=25
    Rotate Z=0
    Scale X=0
    Scale Y=0
    Scale Z=0
    Translate X=0
    Translate Y=-7
    Translate Z=-5
    Weight=1
   )
   Set Transforms for Bone_1_1_1_4 = RigVMFunctionReferenceNode(
    ExecuteContext=()
    Joint=(Type=Bone,Name="dyn_strapFront_01_l")
    Rotate X=0
    Rotate Y=30
    Rotate Z=-90
    Scale X=0
    Scale Y=0
    Scale Z=0
    Translate X=0
    Translate Y=0
    Translate Z=0
    Weight=@n5.OutputParam
   )
   Set Transforms for Bone_1_1_1_4_1 = RigVMFunctionReferenceNode(
    ExecuteContext=()
    Joint=(Type=Bone,Name="dyn_strapFront_01_r")
    Rotate X=10
    Rotate Y=30
    Rotate Z=-90
    Scale X=0
    Scale Y=0
    Scale Z=0
    Translate X=0
    Translate Y=0
    Translate Z=0
    Weight=@n6.OutputParam
   )
   n8 = &RigUnit_SphericalPoseReader(
    ActiveRegionScaleFactors=(PositiveWidth=3.5,NegativeWidth=1,PositiveHeight=3.5,NegativeHeight=4)
    ActiveRegionSize=0.1
    Debug=(bDrawDebug=True,DebugScale=25,DebugSegments=20,DebugThickness=0.25)
    DriverAxis=(X=-1,Y=0,Z=0)
    DriverItem=(Type=Bone,Name="thigh_r")
    ExecutePin=()
    FalloffRegionScaleFactors=(PositiveWidth=1,NegativeWidth=0.9,PositiveHeight=1,NegativeHeight=1)
    FalloffSize=0.45
    FlipHeightScaling=true
    FlipWidthScaling=true
    OptionalParentItem=(Type=Bone,Name="None")
    RotationOffset=(X=-90,Y=0,Z=0)
   )
   Set Transforms for Bone_1_1_1_4_2 = RigVMFunctionReferenceNode(
    ExecuteContext=()
    Joint=(Type=Bone,Name="dyn_strapFront_01_r")
    Rotate X=0
    Rotate Y=30
    Rotate Z=90
    Scale X=0
    Scale Y=0
    Scale Z=0
    Translate X=-12
    Translate Y=0
    Translate Z=0
    Weight=@n8.OutputParam
   )
   Set Transforms for Bone_1_1_1_4_3_1 = RigVMFunctionReferenceNode(
    ExecuteContext=()
    Joint=(Type=Bone,Name="dyn_strapFront_01_l")
    Rotate X=0
    Rotate Y=0
    Rotate Z=5
    Scale X=0
    Scale Y=0
    Scale Z=0
    Translate X=0
    Translate Y=0
    Translate Z=0
    Weight=1
   )
   &RigUnit_SphericalPoseReader(
    ActiveRegionScaleFactors=(PositiveWidth=3.5,NegativeWidth=0.9,PositiveHeight=3.5,NegativeHeight=4)
    ActiveRegionSize=0.1
    Debug=(bDrawDebug=True,DebugScale=25,DebugSegments=20,DebugThickness=0.25)
    DriverAxis=(X=1,Y=0,Z=0)
    DriverItem=(Type=Bone,Name="thigh_r")
    ExecutePin=()
    FalloffRegionScaleFactors=(PositiveWidth=1,NegativeWidth=0.9,PositiveHeight=1,NegativeHeight=1)
    FalloffSize=0.45
    FlipHeightScaling=true
    FlipWidthScaling=true
    OptionalParentItem=(Type=Bone,Name="None")
    RotationOffset=(X=-90,Y=0,Z=0)
   )
  ]}
]
hierarchy [
 root {loc=<0 0 0>, rot=<0 0 0 1>}
 pelvis {parent=root, loc=<0 -0.7071 93.1191>, rot=<0.0127 0.707 -0.0127 -0.707>}
 spine_01 {parent=pelvis, loc=<6.7494 -0.6388 0>, rot=<0 0 0.034 -0.9994>}
 spine_02 {parent=spine_01, loc=<7.2518 -0.035 0>, rot=<0 0 0.0238 -0.9997>}
 spine_03 {parent=spine_02, loc=<6.0104 0 0>, rot=<0 0 0.0149 -0.9999>}
 spine_04 {parent=spine_03, loc=<8.1024 0 0>, rot=<0 0 -0.0132 -0.9999>}
 spine_05 {parent=spine_04, loc=<10.0146 0 0>, rot=<0 0 -0.0015 -1>}
 clavicle_l {parent=spine_05, loc=<4.3575 -0.1466 -2.3223>, rot=<0.0465 -0.7289 -0.009 -0.6829>}
 upperarm_l {parent=clavicle_l, loc=<15.5111 0 0>, rot=<0.0226 -0.3234 -0.0088 -0.9459>}
 lowerarm_l {parent=upperarm_l, loc=<26.6514 0 0>, rot=<0.0391 -0.0099 -0.2453 -0.9686>}
 hand_l {parent=lowerarm_l, loc=<22.3491 0 0>, rot=<-0.7081 -0.0507 0.0975 -0.6975>}
 index_metacarpal_l {parent=hand_l, loc=<1.6535 0.0183 -2.96>, rot=<0.0158 -0.0124 -0.0173 -0.9996>}
 index_01_l {parent=index_metacarpal_l, loc=<8.3008 0.0357 -0.0524>, rot=<-0.0044 0.0485 0.0924 -0.9945>}
 index_02_l {parent=index_01_l, loc=<3.3587 -0.3004 0.0452>, rot=<-0.002 0.0027 0.3225 -0.9466>}
 index_03_l {parent=index_02_l, loc=<2.5578 -0.123 -0.0252>, rot=<-0.002 0.0067 0.2558 -0.9667>}
 middle_metacarpal_l {parent=hand_l, loc=<2.1856 -0.1725 -0.5984>, rot=<0.006 0.0168 -0.0329 -0.9993>}
 middle_01_l {parent=middle_metacarpal_l, loc=<7.3275 -0.2447 0.0129>, rot=<-0.0208 0.0042 0.0858 -0.9961>}
 middle_02_l {parent=middle_01_l, loc=<4.0123 0.1401 0.1175>, rot=<-0.0218 0.0258 0.3512 -0.9357>}
 middle_03_l {parent=middle_02_l, loc=<2.6077 0.0534 0.1373>, rot=<0.0082 0.0188 0.1708 -0.9851>}
 pinky_metacarpal_l {parent=hand_l, loc=<1.7616 -0.9753 3.3814>, rot=<-0.0397 0.0814 -0.069 -0.9935>}
 pinky_01_l {parent=pinky_metacarpal_l, loc=<7.1709 -0.4831 -0.1094>, rot=<-0.0654 -0.0035 0.2242 -0.9723>}
 pinky_02_l {parent=pinky_01_l, loc=<2.8274 0.0893 -0.1043>, rot=<0.0037 0.0011 0.2854 -0.9584>}
 pinky_03_l {parent=pinky_02_l, loc=<2.2046 0.0307 -0.0078>, rot=<-0.0002 0.0098 0.1445 -0.9895>}
 ring_metacarpal_l {parent=hand_l, loc=<2.2074 -0.3625 1.3638>, rot=<-0.0284 0.0722 -0.0458 -0.9959>}
 ring_01_l {parent=ring_metacarpal_l, loc=<7.0295 -0.6197 -0.1446>, rot=<-0.0489 -0.0252 0.1892 -0.9804>}
 ring_02_l {parent=ring_01_l, loc=<3.6174 0.4646 -0.1027>, rot=<0 0 0.2806 -0.9598>}
 ring_03_l {parent=ring_02_l, loc=<2.4152 0.1141 0.0199>, rot=<0 0 0.1189 -0.9929>}
 thumb_01_l {parent=hand_l, loc=<1.3118 -0.3191 -2.459>, rot=<0.6546 -0.4042 0.0664 -0.6354>}
 thumb_02_l {parent=thumb_01_l, loc=<5.9336 -0.0448 0.0401>, rot=<-0.0037 0.0213 0.298 -0.9543>}
 thumb_03_l {parent=thumb_02_l, loc=<2.2223 0.3682 -0.0907>, rot=<0.0118 -0.013 0.0066 -0.9998>}
 lowerarm_twist_01_l {parent=lowerarm_l, loc=<17.8549 0 0>, rot=<0 0 0 1>}
 lowerarm_twist_02_l {parent=lowerarm_l, loc=<9.6199 0 0>, rot=<0 0 0 1>}
 upperarm_twist_01_l {parent=upperarm_l, loc=<0.0041 0 0>, rot=<0 0 0 -1>}
 upperarm_twist_02_l {parent=upperarm_l, loc=<13.7941 0 0>, rot=<0 0 0 -1>}
 clavicle_r {parent=spine_05, loc=<4.3575 -0.1466 2.3223>, rot=<0.7289 0.0465 -0.6829 0.009>}
 upperarm_r {parent=clavicle_r, loc=<-15.5111 0 0>, rot=<0.0226 -0.3234 -0.0088 -0.9459>}
 lowerarm_r {parent=upperarm_r, loc=<-26.6514 0 0>, rot=<0.0391 -0.0099 -0.2453 -0.9686>}
 hand_r {parent=lowerarm_r, loc=<-22.3492 0 0>, rot=<-0.7081 -0.0507 0.0975 -0.6975>}
 index_metacarpal_r {parent=hand_r, loc=<-1.6534 -0.0181 2.9601>, rot=<0.0158 -0.0124 -0.0173 -0.9996>}
 index_01_r {parent=index_metacarpal_r, loc=<-8.3008 -0.0357 0.0524>, rot=<-0.0044 0.0485 0.0924 -0.9945>}
 index_02_r {parent=index_01_r, loc=<-3.3589 0.3003 -0.0453>, rot=<-0.002 0.0027 0.3225 -0.9466>}
 index_03_r {parent=index_02_r, loc=<-2.5578 0.1231 0.0251>, rot=<-0.002 0.0067 0.2558 -0.9667>}
 middle_metacarpal_r {parent=hand_r, loc=<-2.1856 0.1726 0.5984>, rot=<0.006 0.0168 -0.0329 -0.9993>}
 middle_01_r {parent=middle_metacarpal_r, loc=<-7.3273 0.2447 -0.0129>, rot=<-0.0208 0.0042 0.0858 -0.9961>}
 middle_02_r {parent=middle_01_r, loc=<-4.0142 -0.1411 -0.07>, rot=<-0.0217 0.0259 0.3512 -0.9357>}
 middle_03_r {parent=middle_02_r, loc=<-2.6074 -0.0534 -0.1312>, rot=<0.0083 0.0188 0.1708 -0.9851>}
 pinky_metacarpal_r {parent=hand_r, loc=<-1.7617 0.9753 -3.3814>, rot=<-0.0397 0.0814 -0.069 -0.9935>}
 pinky_01_r {parent=pinky_metacarpal_r, loc=<-7.1707 0.4832 0.1094>, rot=<-0.0654 -0.0035 0.2242 -0.9723>}
 pinky_02_r {parent=pinky_01_r, loc=<-2.8274 -0.0892 0.1043>, rot=<0.0037 0.0011 0.2854 -0.9584>}
 pinky_03_r {parent=pinky_02_r, loc=<-2.2045 -0.031 0.0078>, rot=<-0.0002 0.0098 0.1445 -0.9895>}
 ring_metacarpal_r {parent=hand_r, loc=<-2.2073 0.3627 -1.3638>, rot=<-0.0284 0.0722 -0.0458 -0.9959>}
 ring_01_r {parent=ring_metacarpal_r, loc=<-7.0297 0.6197 0.1446>, rot=<-0.0489 -0.0252 0.1892 -0.9804>}
 ring_02_r {parent=ring_01_r, loc=<-3.6172 -0.4648 0.1027>, rot=<0 0 0.2806 -0.9598>}
 ring_03_r {parent=ring_02_r, loc=<-2.4152 -0.1141 -0.0199>, rot=<0 0 0.1189 -0.9929>}
 thumb_01_r {parent=hand_r, loc=<-1.3119 0.3192 2.459>, rot=<0.6546 -0.4042 0.0664 -0.6354>}
 thumb_02_r {parent=thumb_01_r, loc=<-5.9335 0.0447 -0.0401>, rot=<-0.0037 0.0213 0.298 -0.9543>}
 thumb_03_r {parent=thumb_02_r, loc=<-2.2223 -0.3682 0.0907>, rot=<0.0118 -0.013 0.0066 -0.9998>}
 lowerarm_twist_01_r {parent=lowerarm_r, loc=<-17.855 0 0>, rot=<0 0 0 -1>}
 lowerarm_twist_02_r {parent=lowerarm_r, loc=<-9.6199 0 0>, rot=<0 0 0 -1>}
 upperarm_twist_01_r {parent=upperarm_r, loc=<-0.0041 0 0>, rot=<0 0 0 1>}
 upperarm_twist_02_r {parent=upperarm_r, loc=<-13.7942 0 0>, rot=<0 0 0 1>}
 neck_01 {parent=spine_05, loc=<11.3138 0.1529 0>, rot=<0 0 -0.2003 -0.9797>}
 neck_02 {parent=neck_01, loc=<4.5627 0 0>, rot=<0 0 0.0504 -0.9987>}
 head {parent=neck_02, loc=<5.3914 0 0>, rot=<0 0 0.115 -0.9934>}
 thigh_l {parent=pelvis, loc=<-2.5693 -0.0109 -7.7501>, rot=<0.0012 0.0572 0.0155 -0.9982>}
 calf_l {parent=thigh_l, loc=<-39.0484 0 0>, rot=<0.0006 -0.0161 -0.0391 -0.9991>}
 calf_twist_01_l {parent=calf_l, loc=<-35.8131 -0.0967 0.0397>, rot=<0.0237 0.0095 -0.0035 -0.9997>}
 calf_twist_02_l {parent=calf_l, loc=<-18.8064 -0.0967 0.0397>, rot=<0.0237 0.0095 -0.0035 -0.9997>}
 foot_l {parent=calf_l, loc=<-42.561 0.1077 0.4292>, rot=<0.0418 -0.0455 0.0343 -0.9975>}
 ball_l {parent=foot_l, loc=<-6.1599 13.8264 0.6806>, rot=<0.0001 -0.0001 -0.7063 -0.7079>}
 thigh_twist_01_l {parent=thigh_l, loc=<-21.6303 0 0>, rot=<0.0013 0 0 -1>}
 thigh_r {parent=pelvis, loc=<-2.5691 -0.0109 7.7501>, rot=<0.0572 -0.0012 0.9982 0.0155>}
 calf_r {parent=thigh_r, loc=<39.0487 0 0>, rot=<0.0006 -0.0161 -0.0391 -0.9991>}
 calf_twist_01_r {parent=calf_r, loc=<35.8132 0.0967 -0.0397>, rot=<0.0237 0.0095 -0.0035 -0.9997>}
 calf_twist_02_r {parent=calf_r, loc=<18.8064 0.0967 -0.0397>, rot=<0.0237 0.0095 -0.0035 -0.9997>}
 foot_r {parent=calf_r, loc=<42.5611 -0.1077 -0.4292>, rot=<0.0418 -0.0455 0.0343 -0.9975>}
 ball_r {parent=foot_r, loc=<6.1598 -13.8264 -0.6806>, rot=<0.0001 -0.0001 -0.7063 -0.7079>}
 thigh_twist_01_r {parent=thigh_r, loc=<21.6303 0 0>, rot=<0.0013 0 0 -1>}
 dyn_strapBack_01_Base_r {parent=pelvis, loc=<0.2205 -11.5593 6.7735>, rot=<-0.0276 -0.9866 0.0643 0.1473>}
 dyn_strapBack_01_r {parent=dyn_strapBack_01_Base_r, loc=<0 0 0>, rot=<0 0 0 -1>}
 dyn_strapBack_02_r {parent=dyn_strapBack_01_r, loc=<10.3748 0 0>, rot=<-0.183 -0.051 -0.0659 -0.9796>}
 dyn_strapBack_03_r {parent=dyn_strapBack_02_r, loc=<4.1074 0 0>, rot=<0.0735 0.0459 0.027 -0.9959>}
 dyn_strapBack_04_r {parent=dyn_strapBack_03_r, loc=<11.1673 0 0>, rot=<-0.0111 -0.0348 -0.0175 -0.9992>}
 dyn_strapBack_05_r {parent=dyn_strapBack_04_r, loc=<9.1259 0 0>, rot=<-0.0455 -0.2772 -0.2119 -0.936>}
 dyn_strapBack_06_r {parent=dyn_strapBack_05_r, loc=<6.2358 0 0>, rot=<-0.0285 -0.0519 -0.1726 -0.9832>}
 dyn_strapBack_07_r {parent=dyn_strapBack_06_r, loc=<3.2508 0 0>, rot=<0.0045 -0.0382 -0.1966 -0.9797>}
 dyn_strapBack_08_r {parent=dyn_strapBack_07_r, loc=<3.1064 0 0>, rot=<-0.0175 0 0 -0.9998>}
 dyn_strapBack_01_Base_l {parent=pelvis, loc=<0.2204 -11.5594 -6.7735>, rot=<0.9866 -0.0276 0.1473 -0.0643>}
 dyn_strapBack_01_l {parent=dyn_strapBack_01_Base_l, loc=<0 0 0>, rot=<0 0 0 -1>}
 dyn_strapBack_02_l {parent=dyn_strapBack_01_l, loc=<-10.3747 0 0>, rot=<-0.183 -0.051 -0.0659 -0.9796>}
 dyn_strapBack_03_l {parent=dyn_strapBack_02_l, loc=<-4.1074 0 0>, rot=<0.0735 0.0459 0.027 -0.9959>}
 dyn_strapBack_04_l {parent=dyn_strapBack_03_l, loc=<-11.1674 0 0>, rot=<-0.0111 -0.0348 -0.0175 -0.9992>}
 dyn_strapBack_05_l {parent=dyn_strapBack_04_l, loc=<-9.1259 0 0>, rot=<-0.0455 -0.2772 -0.2119 -0.936>}
 dyn_strapBack_06_l {parent=dyn_strapBack_05_l, loc=<-6.2359 0 -0.0001>, rot=<-0.0285 -0.0519 -0.1726 -0.9832>}
 dyn_strapBack_07_l {parent=dyn_strapBack_06_l, loc=<-3.2508 0 0.0001>, rot=<0.0045 -0.0382 -0.1966 -0.9797>}
 dyn_strapBack_08_l {parent=dyn_strapBack_07_l, loc=<-3.1064 0 0>, rot=<-0.0175 0 0 -0.9998>}
 dyn_frontFlapMain_Base {parent=pelvis, loc=<1.5492 11.6478 0.0062>, rot=<0 0 0.9999 -0.0157>}
 dyn_frontFlapMain {parent=dyn_frontFlapMain_Base, loc=<0 0 0>, rot=<0 0 0 1>}
 dyn_strapFront_01_r {parent=dyn_frontFlapMain, loc=<2.3027 0.8819 5.5482>, rot=<0.1488 0.032 -0.0032 -0.9883>}
 dyn_strapFront_02_r {parent=dyn_strapFront_01_r, loc=<8.5403 0 0>, rot=<-0.1595 0.1777 0.0735 -0.9683>}
 dyn_strapFront_03_r {parent=dyn_strapFront_02_r, loc=<5.1104 0 0>, rot=<0.0211 -0.0441 -0.0087 -0.9988>}
 dyn_strapFront_04_r {parent=dyn_strapFront_03_r, loc=<9.0743 0 0>, rot=<0.0589 0.0313 -0.0235 -0.9975>}
 dyn_strapFront_05_r {parent=dyn_strapFront_04_r, loc=<9.5626 0 0>, rot=<0.0323 0.2359 -0.1867 -0.9531>}
 dyn_strapFront_06_r {parent=dyn_strapFront_05_r, loc=<8.0904 0 0>, rot=<-0.0042 0.0668 -0.394 -0.9167>}
 dyn_strapFront_07_r {parent=dyn_strapFront_06_r, loc=<3.2899 0 0>, rot=<0.0207 -0.0008 -0.1586 -0.9871>}
 dyn_strapFront_08_r {parent=dyn_strapFront_07_r, loc=<3.1064 0 0>, rot=<0.0175 0 0 -0.9998>}
 dyn_strapFront_01_l {parent=dyn_frontFlapMain, loc=<2.3026 0.8819 -5.5482>, rot=<0.032 -0.1488 0.9883 -0.0032>}
 dyn_strapFront_02_l {parent=dyn_strapFront_01_l, loc=<-8.5403 0 0>, rot=<-0.1595 0.1777 0.0735 -0.9683>}
 dyn_strapFront_03_l {parent=dyn_strapFront_02_l, loc=<-5.1105 0 0>, rot=<0.0211 -0.0441 -0.0087 -0.9988>}
 dyn_strapFront_04_l {parent=dyn_strapFront_03_l, loc=<-9.0743 0 0>, rot=<0.0589 0.0313 -0.0235 -0.9975>}
 dyn_strapFront_05_l {parent=dyn_strapFront_04_l, loc=<-9.5626 0 0>, rot=<0.0323 0.2359 -0.1867 -0.9531>}
 dyn_strapFront_06_l {parent=dyn_strapFront_05_l, loc=<-8.0904 0 0>, rot=<-0.0042 0.0668 -0.394 -0.9167>}
 dyn_strapFront_07_l {parent=dyn_strapFront_06_l, loc=<-3.2899 0 0>, rot=<0.0207 -0.0008 -0.1586 -0.9871>}
 dyn_strapFront_08_l {parent=dyn_strapFront_07_l, loc=<-3.1064 0.0001 0>, rot=<0.0175 0 0 -0.9998>}
]
