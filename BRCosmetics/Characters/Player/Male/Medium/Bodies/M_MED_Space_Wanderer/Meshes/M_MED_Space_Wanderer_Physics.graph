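# Physics asset for the Space Wanderer medium male body, giving physics bodies to a front flap and front and back straps on the right leg.
# Two seven-link capsule chains hang from pelvis-constrained base boxes, the front one via the flap, with per-joint swing and twist limits.
# A tight constraint joins both chain tips, closing the straps into a loop.
# Pelvis, multi-capsule right thigh and freely rotating right calf bodies supply the body shapes.
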PhysicsAsset /BRCosmetics/Characters/Player/Male/Medium/Bodies/M_MED_Space_Wanderer/Meshes/M_MED_Space_Wanderer_Physics
bodies {
 calf_r [{shapeType=Capsule, length=23.2739, radius=7.5275, center=<24.161 -2.1754 0.5764>, rotation=<rot -89.1556 99.1968 -95.9676>}]
 dyn_frontFlapMain [{shapeType=Capsule, length=4, radius=8, shapeName=dyn_strapFront_07_l_capsule, center=<4 -6 0>, rotation=<rot 90 0 0>}]
 dyn_frontFlapMain_Base [{shapeType=Box, extentX=1.01, extentY=1.01, extentZ=1.01, rotation=<rot 0 -90 0>}]
 dyn_strapBack_01_Base_r [{shapeType=Box, extentX=2.013, extentY=4.1498, extentZ=2.0838, rotation=<rot 0 -90 0>}]
 dyn_strapBack_01_r [{shapeType=Capsule, length=4, radius=8, shapeName=dyn_strapFront_07_l_capsule, center=<4 -6 0>, rotation=<rot 90 0 0>}]
 dyn_strapBack_02_r [{shapeType=Capsule, length=4, radius=8, shapeName=dyn_strapFront_07_l_capsule, center=<4 -6 0>, rotation=<rot 90 0 0>}]
 dyn_strapBack_03_r [{shapeType=Capsule, length=4, radius=8, shapeName=dyn_strapFront_07_l_capsule, center=<4 -6 0>, rotation=<rot 90 0 0>}]
 dyn_strapBack_04_r [{shapeType=Capsule, length=4, radius=8, shapeName=dyn_strapFront_07_l_capsule, center=<4 -6 0>, rotation=<rot 90 0 0>}]
 dyn_strapBack_05_r [{shapeType=Capsule, length=4, radius=8, shapeName=dyn_strapFront_07_l_capsule, center=<4 -6 0>, rotation=<rot 90 0 0>}]
 dyn_strapBack_06_r [{shapeType=Capsule, length=4, radius=8, shapeName=dyn_strapFront_07_l_capsule, center=<4 -6 0>, rotation=<rot 90 0 0>}]
 dyn_strapBack_07_r [{shapeType=Capsule, length=4, radius=8, shapeName=dyn_strapFront_07_l_capsule, center=<4 -6 0>, rotation=<rot 90 0 0>}]
 dyn_strapFront_01_r [{shapeType=Capsule, length=4, radius=8, shapeName=dyn_strapFront_07_l_capsule, center=<4 -6 0>, rotation=<rot 90 0 0>}]
 dyn_strapFront_02_r [{shapeType=Capsule, length=4, radius=8, shapeName=dyn_strapFront_07_l_capsule, center=<4 -6 0>, rotation=<rot 90 0 0>}]
 dyn_strapFront_03_r [{shapeType=Capsule, length=4, radius=8, shapeName=dyn_strapFront_07_l_capsule, center=<4 -6 0>, rotation=<rot 90 0 0>}]
 dyn_strapFront_04_r [{shapeType=Capsule, length=4, radius=8, shapeName=dyn_strapFront_07_l_capsule, center=<4 -6 0>, rotation=<rot 90 0 0>}]
 dyn_strapFront_05_r [{shapeType=Capsule, length=4, radius=8, shapeName=dyn_strapFront_07_l_capsule, center=<4 -6 0>, rotation=<rot 90 0 0>}]
 dyn_strapFront_06_r [{shapeType=Capsule, length=4, radius=8, shapeName=dyn_strapFront_07_l_capsule, center=<4 -6 0>, rotation=<rot 90 0 0>}]
 dyn_strapFront_07_r [{shapeType=Capsule, length=4, radius=8, shapeName=dyn_strapFront_07_l_capsule, center=<4 -6 0>, rotation=<rot 90 0 0>}]
 pelvis [{shapeType=Capsule, length=9.2317, radius=9.323, center=<-4.6744 -0.7351 0.0287>, rotation=<rot -0.1097 -90 0.2601>}]
 thigh_twist_01_r [
  {shapeType=Capsule, length=33.5186, radius=7.6, center=<13.9904 -1.2371 3.9325>, rotation=<rot -88.0327 163.4606 -160.1909>}
  {shapeType=Capsule, length=52.4942, radius=8.3721, center=<4.2249 -0.0012 0.5437>, rotation=<rot 89.8321 -0.0119 -0.012>}
  {shapeType=Capsule, length=51.0835, radius=7.7791, center=<10.4988 -0.0006 -4.893>, rotation=<rot -84.2551 -179.9996 -180>}
  {
   shapeType=Capsule
   length=50.0691
   radius=8.7935
   shapeName=thigh_twist_01_l_capsule
   center=<13.1209 -0.0007 -14.7165>
   rotation=<rot -84.2551 -179.9996 -180>
  }
 ]
}
constraints [
 {bone1Name=dyn_frontFlapMain, bone2Name=dyn_frontFlapMain_Base, swing1LimitDegrees=45, swing2LimitDegrees=45, twistLimitDegrees=10}
 {bone1Name=dyn_strapFront_01_r, bone2Name=dyn_frontFlapMain, swing1LimitDegrees=20, swing2LimitDegrees=20, twistLimitDegrees=1}
 {bone1Name=dyn_strapFront_02_r, bone2Name=dyn_strapFront_01_r, swing1LimitDegrees=45, swing2LimitDegrees=45, twistLimitDegrees=10}
 {bone1Name=dyn_strapFront_03_r, bone2Name=dyn_strapFront_02_r, swing1LimitDegrees=45, swing2LimitDegrees=45, twistLimitDegrees=10}
 {bone1Name=dyn_strapFront_04_r, bone2Name=dyn_strapFront_03_r, swing1LimitDegrees=45, swing2LimitDegrees=45, twistLimitDegrees=10}
 {bone1Name=dyn_strapFront_05_r, bone2Name=dyn_strapFront_04_r, swing1LimitDegrees=45, swing2LimitDegrees=45, twistLimitDegrees=10}
 {bone1Name=dyn_strapFront_06_r, bone2Name=dyn_strapFront_05_r, swing1LimitDegrees=45, swing2LimitDegrees=45, twistLimitDegrees=10}
 {bone1Name=dyn_strapFront_07_r, bone2Name=dyn_strapFront_06_r, swing1LimitDegrees=5, swing2LimitDegrees=5, twistLimitDegrees=2}
 {bone1Name=dyn_strapBack_01_r, bone2Name=dyn_strapBack_01_Base_r, swing1LimitDegrees=45, swing2LimitDegrees=20, twistLimitDegrees=1}
 {bone1Name=dyn_strapBack_02_r, bone2Name=dyn_strapBack_01_r, swing1LimitDegrees=5, swing2LimitDegrees=5, twistLimitDegrees=10}
 {bone1Name=dyn_strapBack_03_r, bone2Name=dyn_strapBack_02_r, swing1LimitDegrees=10, swing2LimitDegrees=10, twistLimitDegrees=10}
 {bone1Name=dyn_strapBack_04_r, bone2Name=dyn_strapBack_03_r, swing1LimitDegrees=45, swing2LimitDegrees=45, twistLimitDegrees=10}
 {bone1Name=dyn_strapBack_05_r, bone2Name=dyn_strapBack_04_r, swing1LimitDegrees=45, swing2LimitDegrees=45, twistLimitDegrees=10}
 {bone1Name=dyn_strapBack_06_r, bone2Name=dyn_strapBack_05_r, swing1LimitDegrees=45, swing2LimitDegrees=45, twistLimitDegrees=10}
 {bone1Name=dyn_strapBack_07_r, bone2Name=dyn_strapBack_06_r, swing1LimitDegrees=5, swing2LimitDegrees=5, twistLimitDegrees=2}
 {bone1Name=thigh_twist_01_r, bone2Name=pelvis, swing1LimitDegrees=35, swing2LimitDegrees=35, twistLimitDegrees=5}
 {bone1Name=dyn_strapBack_01_Base_r, bone2Name=pelvis, swing1LimitDegrees=45, swing2LimitDegrees=45, twistLimitDegrees=10}
 {bone1Name=dyn_frontFlapMain_Base, bone2Name=pelvis, swing1LimitDegrees=45, swing2LimitDegrees=45, twistLimitDegrees=10}
 {bone1Name=dyn_strapFront_07_r, bone2Name=dyn_strapBack_07_r, swing1LimitDegrees=1, swing2LimitDegrees=1, twistLimitDegrees=1}
 {
  bone1Name=calf_r
  bone2Name=pelvis
  swing1LimitDegrees=45
  swing1Motion=Free
  swing2LimitDegrees=45
  swing2Motion=Free
  twistLimitDegrees=45
  twistMotion=Free
 }
]
