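# Physics asset for the male medium Space Wanderer player body, covering its dangling front and back straps.
# Identical small capsules form four bone chains, six links each except the eight-link left back strap.
# Constraints join neighboring links with free swing and twist, so the stored angle limits are inactive.
# Each thigh has an offset capsule, and no constraint ties thighs or chain roots to other bodies.
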PhysicsAsset /BRCosmetics/Characters/Player/Male/Medium/Bodies/M_MED_Space_Wanderer/Meshes/M_MED_Space_Wanderer_Physics1
bodies {
 dyn_strapBack_01_l [{shapeType=Capsule, length=4, radius=4, shapeName=dyn_strapFront_04_r_capsule, rotation=<rot 90 0 0>}]
 dyn_strapBack_01_r [{shapeType=Capsule, length=4, radius=4, shapeName=dyn_strapFront_04_r_capsule, rotation=<rot 90 0 0>}]
 dyn_strapBack_02_l [{shapeType=Capsule, length=4, radius=4, shapeName=dyn_strapFront_04_r_capsule, rotation=<rot 90 0 0>}]
 dyn_strapBack_02_r [{shapeType=Capsule, length=4, radius=4, shapeName=dyn_strapFront_04_r_capsule, rotation=<rot 90 0 0>}]
 dyn_strapBack_03_l [{shapeType=Capsule, length=4, radius=4, shapeName=dyn_strapFront_04_r_capsule, rotation=<rot 90 0 0>}]
 dyn_strapBack_03_r [{shapeType=Capsule, length=4, radius=4, shapeName=dyn_strapFront_04_r_capsule, rotation=<rot 90 0 0>}]
 dyn_strapBack_04_l [{shapeType=Capsule, length=4, radius=4, shapeName=dyn_strapFront_04_r_capsule, rotation=<rot 90 0 0>}]
 dyn_strapBack_04_r [{shapeType=Capsule, length=4, radius=4, shapeName=dyn_strapFront_04_r_capsule, rotation=<rot 90 0 0>}]
 dyn_strapBack_05_l [{shapeType=Capsule, length=4, radius=4, shapeName=dyn_strapFront_04_r_capsule, rotation=<rot 90 0 0>}]
 dyn_strapBack_05_r [{shapeType=Capsule, length=4, radius=4, shapeName=dyn_strapFront_04_r_capsule, rotation=<rot 90 0 0>}]
 dyn_strapBack_06_l [{shapeType=Capsule, length=4, radius=4, shapeName=dyn_strapFront_04_r_capsule, rotation=<rot 90 0 0>}]
 dyn_strapBack_06_r [{shapeType=Capsule, length=4, radius=4, shapeName=dyn_strapFront_04_r_capsule, rotation=<rot 90 0 0>}]
 dyn_strapBack_07_l [{shapeType=Capsule, length=4, radius=4, shapeName=dyn_strapFront_04_r_capsule, rotation=<rot 90 0 0>}]
 dyn_strapBack_08_l [{shapeType=Capsule, length=4, radius=4, shapeName=dyn_strapFront_04_r_capsule, rotation=<rot 90 0 0>}]
 dyn_strapFront_01_l [{shapeType=Capsule, length=4, radius=4, shapeName=dyn_strapFront_04_r_capsule, rotation=<rot 90 0 0>}]
 dyn_strapFront_01_r [{shapeType=Capsule, length=4, radius=4, shapeName=dyn_strapFront_04_r_capsule, rotation=<rot 90 0 0>}]
 dyn_strapFront_02_l [{shapeType=Capsule, length=4, radius=4, shapeName=dyn_strapFront_04_r_capsule, rotation=<rot 90 0 0>}]
 dyn_strapFront_02_r [{shapeType=Capsule, length=4, radius=4, shapeName=dyn_strapFront_04_r_capsule, rotation=<rot 90 0 0>}]
 dyn_strapFront_03_l [{shapeType=Capsule, length=4, radius=4, shapeName=dyn_strapFront_04_r_capsule, rotation=<rot 90 0 0>}]
 dyn_strapFront_03_r [{shapeType=Capsule, length=4, radius=4, shapeName=dyn_strapFront_04_r_capsule, rotation=<rot 90 0 0>}]
 dyn_strapFront_04_l [{shapeType=Capsule, length=4, radius=4, shapeName=dyn_strapFront_04_r_capsule, rotation=<rot 90 0 0>}]
 dyn_strapFront_04_r [{shapeType=Capsule, length=4, radius=4, shapeName=dyn_strapFront_04_r_capsule, rotation=<rot 90 0 0>}]
 dyn_strapFront_05_l [{shapeType=Capsule, length=4, radius=4, shapeName=dyn_strapFront_04_r_capsule, rotation=<rot 90 0 0>}]
 dyn_strapFront_05_r [{shapeType=Capsule, length=4, radius=4, shapeName=dyn_strapFront_04_r_capsule, rotation=<rot 90 0 0>}]
 dyn_strapFront_06_l [{shapeType=Capsule, length=4, radius=4, shapeName=dyn_strapFront_04_r_capsule, rotation=<rot 90 0 0>}]
 dyn_strapFront_06_r [{shapeType=Capsule, length=4, radius=4, shapeName=dyn_strapFront_04_r_capsule, rotation=<rot 90 0 0>}]
 thigh_l [{shapeType=Capsule, length=5, radius=5, center=<-11.0531 0 0>, rotation=<rot 90 0 0>}]
 thigh_r [{shapeType=Capsule, length=5, radius=5, center=<10.4637 0 0>, rotation=<rot 90 0 0>}]
}
constraints [
 {
  bone1Name=dyn_strapBack_06_r
  bone2Name=dyn_strapBack_05_r
  swing1LimitDegrees=45
  swing1Motion=Free
  swing2LimitDegrees=45
  swing2Motion=Free
  twistLimitDegrees=45
  twistMotion=Free
 }
 {
  bone1Name=dyn_strapBack_05_r
  bone2Name=dyn_strapBack_04_r
  swing1LimitDegrees=45
  swing1Motion=Free
  swing2LimitDegrees=45
  swing2Motion=Free
  twistLimitDegrees=45
  twistMotion=Free
 }
 {
  bone1Name=dyn_strapBack_04_r
  bone2Name=dyn_strapBack_03_r
  swing1LimitDegrees=45
  swing1Motion=Free
  swing2LimitDegrees=45
  swing2Motion=Free
  twistLimitDegrees=45
  twistMotion=Free
 }
 {
  bone1Name=dyn_strapBack_03_r
  bone2Name=dyn_strapBack_02_r
  swing1LimitDegrees=45
  swing1Motion=Free
  swing2LimitDegrees=45
  swing2Motion=Free
  twistLimitDegrees=45
  twistMotion=Free
 }
 {
  bone1Name=dyn_strapBack_02_r
  bone2Name=dyn_strapBack_01_r
  swing1LimitDegrees=45
  swing1Motion=Free
  swing2LimitDegrees=45
  swing2Motion=Free
  twistLimitDegrees=45
  twistMotion=Free
 }
 {
  bone1Name=dyn_strapBack_02_l
  bone2Name=dyn_strapBack_01_l
  swing1LimitDegrees=45
  swing1Motion=Free
  swing2LimitDegrees=45
  swing2Motion=Free
  twistLimitDegrees=45
  twistMotion=Free
 }
 {
  bone1Name=dyn_strapBack_03_l
  bone2Name=dyn_strapBack_02_l
  swing1LimitDegrees=45
  swing1Motion=Free
  swing2LimitDegrees=45
  swing2Motion=Free
  twistLimitDegrees=45
  twistMotion=Free
 }
 {
  bone1Name=dyn_strapBack_04_l
  bone2Name=dyn_strapBack_03_l
  swing1LimitDegrees=45
  swing1Motion=Free
  swing2LimitDegrees=45
  swing2Motion=Free
  twistLimitDegrees=45
  twistMotion=Free
 }
 {
  bone1Name=dyn_strapBack_05_l
  bone2Name=dyn_strapBack_04_l
  swing1LimitDegrees=45
  swing1Motion=Free
  swing2LimitDegrees=45
  swing2Motion=Free
  twistLimitDegrees=45
  twistMotion=Free
 }
 {
  bone1Name=dyn_strapBack_06_l
  bone2Name=dyn_strapBack_05_l
  swing1LimitDegrees=45
  swing1Motion=Free
  swing2LimitDegrees=45
  swing2Motion=Free
  twistLimitDegrees=45
  twistMotion=Free
 }
 {
  bone1Name=dyn_strapBack_07_l
  bone2Name=dyn_strapBack_06_l
  swing1LimitDegrees=45
  swing1Motion=Free
  swing2LimitDegrees=45
  swing2Motion=Free
  twistLimitDegrees=45
  twistMotion=Free
 }
 {
  bone1Name=dyn_strapBack_08_l
  bone2Name=dyn_strapBack_07_l
  swing1LimitDegrees=45
  swing1Motion=Free
  swing2LimitDegrees=45
  swing2Motion=Free
  twistLimitDegrees=45
  twistMotion=Free
 }
 {
  bone1Name=dyn_strapFront_03_r
  bone2Name=dyn_strapFront_02_r
  swing1LimitDegrees=45
  swing1Motion=Free
  swing2LimitDegrees=45
  swing2Motion=Free
  twistLimitDegrees=45
  twistMotion=Free
 }
 {
  bone1Name=dyn_strapFront_04_r
  bone2Name=dyn_strapFront_03_r
  swing1LimitDegrees=45
  swing1Motion=Free
  swing2LimitDegrees=45
  swing2Motion=Free
  twistLimitDegrees=45
  twistMotion=Free
 }
 {
  bone1Name=dyn_strapFront_06_r
  bone2Name=dyn_strapFront_05_r
  swing1LimitDegrees=45
  swing1Motion=Free
  swing2LimitDegrees=45
  swing2Motion=Free
  twistLimitDegrees=45
  twistMotion=Free
 }
 {
  bone1Name=dyn_strapFront_05_r
  bone2Name=dyn_strapFront_04_r
  swing1LimitDegrees=45
  swing1Motion=Free
  swing2LimitDegrees=45
  swing2Motion=Free
  twistLimitDegrees=45
  twistMotion=Free
 }
 {
  bone1Name=dyn_strapFront_02_r
  bone2Name=dyn_strapFront_01_r
  swing1LimitDegrees=45
  swing1Motion=Free
  swing2LimitDegrees=45
  swing2Motion=Free
  twistLimitDegrees=45
  twistMotion=Free
 }
 {
  bone1Name=dyn_strapFront_02_l
  bone2Name=dyn_strapFront_01_l
  swing1LimitDegrees=45
  swing1Motion=Free
  swing2LimitDegrees=45
  swing2Motion=Free
  twistLimitDegrees=45
  twistMotion=Free
 }
 {
  bone1Name=dyn_strapFront_03_l
  bone2Name=dyn_strapFront_02_l
  swing1LimitDegrees=45
  swing1Motion=Free
  swing2LimitDegrees=45
  swing2Motion=Free
  twistLimitDegrees=45
  twistMotion=Free
 }
 {
  bone1Name=dyn_strapFront_04_l
  bone2Name=dyn_strapFront_03_l
  swing1LimitDegrees=45
  swing1Motion=Free
  swing2LimitDegrees=45
  swing2Motion=Free
  twistLimitDegrees=45
  twistMotion=Free
 }
 {
  bone1Name=dyn_strapFront_05_l
  bone2Name=dyn_strapFront_04_l
  swing1LimitDegrees=45
  swing1Motion=Free
  swing2LimitDegrees=45
  swing2Motion=Free
  twistLimitDegrees=45
  twistMotion=Free
 }
 {
  bone1Name=dyn_strapFront_06_l
  bone2Name=dyn_strapFront_05_l
  swing1LimitDegrees=45
  swing1Motion=Free
  swing2LimitDegrees=45
  swing2Motion=Free
  twistLimitDegrees=45
  twistMotion=Free
 }
]
