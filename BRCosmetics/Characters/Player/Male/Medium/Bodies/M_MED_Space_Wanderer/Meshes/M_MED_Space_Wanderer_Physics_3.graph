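# Physics asset for the male medium Space Wanderer body, covering the dynamic front flap and left-side strap bones.
# Capsule chains for the front strap, rooted on the flap, and back strap, rooted on its own base, both trace back to the pelvis.
# A near-rigid constraint joins the two strap tips, closing a loop.
# Pelvis and multi-capsule left thigh bodies provide leg shapes, with the thigh's angular motion left free.
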PhysicsAsset /BRCosmetics/Characters/Player/Male/Medium/Bodies/M_MED_Space_Wanderer/Meshes/M_MED_Space_Wanderer_Physics_3
bodies {
 dyn_frontFlapMain [{shapeType=Capsule, length=4, radius=8, shapeName=dyn_strapFront_07_l_capsule, center=<-4 6 0>, rotation=<rot -0.7908 0 0>}]
 dyn_frontFlapMain_Base [{shapeType=Box, extentX=1.01, extentY=1.01, extentZ=1.01, rotation=<rot 0 -90 0>}]
 dyn_strapBack_01_Base_l [{shapeType=Box, extentX=2.013, extentY=4.1498, extentZ=2.0838, rotation=<rot 0 -90 0>}]
 dyn_strapBack_01_l [{shapeType=Capsule, length=4, radius=8, shapeName=dyn_strapFront_07_l_capsule, center=<-4 6 0>, rotation=<rot 90 0 0>}]
 dyn_strapBack_02_l [{shapeType=Capsule, length=4, radius=8, shapeName=dyn_strapFront_07_l_capsule, center=<-4 6 0>, rotation=<rot 90 0 0>}]
 dyn_strapBack_03_l [{shapeType=Capsule, length=6, radius=8, shapeName=dyn_strapFront_07_l_capsule, center=<-4 6 0>, rotation=<rot 90 0 0>}]
 dyn_strapBack_04_l [{shapeType=Capsule, length=6, radius=8, shapeName=dyn_strapFront_07_l_capsule, center=<-4 6 0>, rotation=<rot 90 0 0>}]
 dyn_strapBack_05_l [{shapeType=Capsule, length=6, radius=8, shapeName=dyn_strapFront_07_l_capsule, center=<-4 6 0>, rotation=<rot 90 0 0>}]
 dyn_strapBack_06_l [{shapeType=Capsule, length=6, radius=8, shapeName=dyn_strapFront_07_l_capsule, center=<-4 6 0>, rotation=<rot 90 0 -13.5871>}]
 dyn_strapBack_07_l [{shapeType=Capsule, length=6, radius=8, shapeName=dyn_strapFront_07_l_capsule, center=<-4 6 0>, rotation=<rot 90 0 0>}]
 dyn_strapFront_01_l [{shapeType=Capsule, length=4, radius=8, shapeName=dyn_strapFront_07_l_capsule, center=<-4 6 0>, rotation=<rot 90 0 0>}]
 dyn_strapFront_02_l [{shapeType=Capsule, length=6, radius=8, shapeName=dyn_strapFront_07_l_capsule, center=<-4 7 0>, rotation=<rot 90 0 0>}]
 dyn_strapFront_03_l [{shapeType=Capsule, length=6, radius=8, shapeName=dyn_strapFront_07_l_capsule, center=<-4 6 0>, rotation=<rot 90 0 0>}]
 dyn_strapFront_04_l [{shapeType=Capsule, length=6, radius=8, shapeName=dyn_strapFront_07_l_capsule, center=<-4 6 0>, rotation=<rot 90 0 0>}]
 dyn_strapFront_05_l [{shapeType=Capsule, length=6, radius=8, shapeName=dyn_strapFront_07_l_capsule, center=<-4 6 0>, rotation=<rot 90 0 0>}]
 dyn_strapFront_06_l [{shapeType=Capsule, length=6, radius=8, shapeName=dyn_strapFront_07_l_capsule, center=<-4 6 0>, rotation=<rot 90 0 0>}]
 dyn_strapFront_07_l [{shapeType=Capsule, length=6, radius=8, shapeName=dyn_strapFront_07_l_capsule, center=<-4 6 0>, rotation=<rot 90 0 0>}]
 pelvis [{shapeType=Capsule, length=9.2317, radius=9.323, center=<-4.6744 -0.7351 0.0287>, rotation=<rot -0.1097 -90 0.2601>}]
 thigh_twist_01_l [
  {shapeType=Capsule, length=25.319, radius=7.6, center=<-11.467 0.097 -2.6212>, rotation=<rot 79.4291 -66.9829 -62.1031>}
  {shapeType=Capsule, length=54.9061, radius=7.2558, center=<-5.0515 -0.001 -0.0107>, rotation=<rot -89.2624 -179.9999 -179.9996>}
  {shapeType=Capsule, length=51.8539, radius=7.0088, center=<-5.5413 -0.0007 6.8903>, rotation=<rot -84.2551 -179.9996 -180>}
  {
   shapeType=Capsule
   length=50.5056
   radius=8.357
   shapeName=thigh_twist_01_l_capsule
   center=<-9.9211 -0.0008 17.1918>
   rotation=<rot -84.2551 -179.9996 -180>
  }
 ]
}
constraints [
 {bone1Name=dyn_frontFlapMain, bone2Name=dyn_frontFlapMain_Base, swing1LimitDegrees=45, swing2LimitDegrees=45, twistLimitDegrees=20}
 {bone1Name=dyn_frontFlapMain_Base, bone2Name=pelvis, swing1LimitDegrees=45, swing2LimitDegrees=45, twistLimitDegrees=20}
 {
  bone1Name=thigh_twist_01_l
  bone2Name=pelvis
  swing1LimitDegrees=45
  swing1Motion=Free
  swing2LimitDegrees=45
  swing2Motion=Free
  twistLimitDegrees=45
  twistMotion=Free
 }
 {bone1Name=dyn_strapBack_01_Base_l, bone2Name=pelvis, swing1LimitDegrees=45, swing2LimitDegrees=45, twistLimitDegrees=20}
 {bone1Name=dyn_strapBack_01_l, bone2Name=dyn_strapBack_01_Base_l, swing1LimitDegrees=45, swing2LimitDegrees=45, twistLimitDegrees=20}
 {bone1Name=dyn_strapBack_02_l, bone2Name=dyn_strapBack_01_l, swing1LimitDegrees=5, swing2LimitDegrees=5, twistLimitDegrees=20}
 {bone1Name=dyn_strapBack_03_l, bone2Name=dyn_strapBack_02_l, swing1LimitDegrees=10, swing2LimitDegrees=10, twistLimitDegrees=20}
 {bone1Name=dyn_strapBack_04_l, bone2Name=dyn_strapBack_03_l, swing1LimitDegrees=45, swing2LimitDegrees=45, twistLimitDegrees=20}
 {bone1Name=dyn_strapBack_05_l, bone2Name=dyn_strapBack_04_l, swing1LimitDegrees=45, swing2LimitDegrees=45, twistLimitDegrees=20}
 {bone1Name=dyn_strapBack_06_l, bone2Name=dyn_strapBack_05_l, swing1LimitDegrees=45, swing2LimitDegrees=45, twistLimitDegrees=20}
 {bone1Name=dyn_strapBack_07_l, bone2Name=dyn_strapBack_06_l, swing1LimitDegrees=10, swing2LimitDegrees=10, twistLimitDegrees=20}
 {bone1Name=dyn_strapFront_01_l, bone2Name=dyn_frontFlapMain, swing1LimitDegrees=45, swing2LimitDegrees=45, twistLimitDegrees=20}
 {bone1Name=dyn_strapFront_07_l, bone2Name=dyn_strapFront_06_l, swing1LimitDegrees=5, swing2LimitDegrees=45, twistLimitDegrees=20}
 {bone1Name=dyn_strapFront_06_l, bone2Name=dyn_strapFront_05_l, swing1LimitDegrees=5, swing2LimitDegrees=5, twistLimitDegrees=20}
 {bone1Name=dyn_strapFront_05_l, bone2Name=dyn_strapFront_04_l, swing1LimitDegrees=45, swing2LimitDegrees=45, twistLimitDegrees=20}
 {bone1Name=dyn_strapFront_04_l, bone2Name=dyn_strapFront_03_l, swing1LimitDegrees=45, swing2LimitDegrees=45, twistLimitDegrees=20}
 {bone1Name=dyn_strapFront_03_l, bone2Name=dyn_strapFront_02_l, swing1LimitDegrees=45, swing2LimitDegrees=45, twistLimitDegrees=20}
 {bone1Name=dyn_strapFront_02_l, bone2Name=dyn_strapFront_01_l, swing1LimitDegrees=10, swing2LimitDegrees=10, twistLimitDegrees=20}
 {bone1Name=dyn_strapBack_07_l, bone2Name=dyn_strapFront_07_l, swing1LimitDegrees=1, swing2LimitDegrees=1, twistLimitDegrees=1}
]
