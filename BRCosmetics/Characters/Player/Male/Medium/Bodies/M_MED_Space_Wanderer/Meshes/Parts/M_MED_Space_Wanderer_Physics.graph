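# Physics asset for the M_MED_Space_Wanderer male player body cosmetic, covering only the head and a dynamic hair bone.
# Two capsule bodies represent the head and dyn_hair_top, with the hair capsule at zero length so it acts as a sphere.
# A single constraint joins the hair to the head with narrow, asymmetric swing limits and almost no twist, keeping any hair motion tightly bounded.
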PhysicsAsset /BRCosmetics/Characters/Player/Male/Medium/Bodies/M_MED_Space_Wanderer/Meshes/Parts/M_MED_Space_Wanderer_Physics
bodies {
 dyn_hair_top [{shapeType=Capsule, length=0, radius=2.2836, center=<0 -1.2628 0.468>, rotation=<rot 90 0 0>}]
 head [{shapeType=Capsule, length=3.5065, radius=4.5619, center=<8.6127 1.6649 0.0486>, rotation=<rot 42.8746 89.8532 -90>}]
}
constraints [{bone1Name=dyn_hair_top, bone2Name=head, swing1LimitDegrees=10, swing2LimitDegrees=30, twistLimitDegrees=1}]
